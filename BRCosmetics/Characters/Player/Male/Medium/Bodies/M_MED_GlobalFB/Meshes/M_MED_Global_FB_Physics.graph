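# Physics asset for the M_MED_Global_FB male medium player body cosmetic, covering its dynamic shorts and zipper bones.
# Capsule bodies on the left and right shorts bones are each constrained to the matching thigh body with tight swing and tighter twist limits.
# The thighs carry only tiny placeholder-sized capsules as anchors, and the zipper master has a small box body with no constraint.
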PhysicsAsset /BRCosmetics/Characters/Player/Male/Medium/Bodies/M_MED_GlobalFB/Meshes/M_MED_Global_FB_Physics
bodies {
 Dyn_Shorts_L [{shapeType=Capsule, length=7.75, radius=8.4375, center=<-4.8641 0.0001 0>, rotation=<rot 90 0 0>}]
 Dyn_Shorts_R [{shapeType=Capsule, length=6.6562, radius=8.6562, center=<5.3198 0 0>, rotation=<rot 90 0 0>}]
 Dyn_zipper_Master [{shapeType=Box, extentX=1, extentY=1, extentZ=1}]
 thigh_l [{shapeType=Capsule, length=1, radius=1}]
 thigh_r [{shapeType=Capsule, length=1, radius=1}]
}
constraints [
 {bone1Name=Dyn_Shorts_L, bone2Name=thigh_l, swing1LimitDegrees=10, swing2LimitDegrees=10, twistLimitDegrees=3}
 {bone1Name=Dyn_Shorts_R, bone2Name=thigh_r, swing1LimitDegrees=10, swing2LimitDegrees=10, twistLimitDegrees=3}
]
