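# Physics asset for the left-leg mesh of the GlobalFB medium male player body cosmetic, covering its dynamic shorts and zipper bones.
# A capsule body on the left shorts bone is constrained to the left thigh, with narrow swing and tighter twist limits.
# The thigh has a tiny capsule and the zipper master bone a small box, which has no constraint.
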PhysicsAsset /BRCosmetics/Characters/Player/Male/Medium/Bodies/M_MED_GlobalFB/Meshes/M_MED_Global_FB_leg_l_Physics
bodies {
 Dyn_Shorts_L [{shapeType=Capsule, length=7.75, radius=8.4375, center=<-4.8641 0.0001 0>, rotation=<rot 90 0 0>}]
 Dyn_zipper_Master [{shapeType=Box, extentX=1, extentY=1, extentZ=1}]
 thigh_l [{shapeType=Capsule, length=1, radius=1}]
}
constraints [{bone1Name=Dyn_Shorts_L, bone2Name=thigh_l, swing1LimitDegrees=15, swing2LimitDegrees=15, twistLimitDegrees=2}]
